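# Physics asset for the front end of the male Assemble_P pickaxe, covering its two-handed handle and dynamic "hoose" (likely hose) bones.
# A long capsule represents the handle, with smaller capsules on one claw bone and three handle hose bones.
# Each small body is constrained directly to the handle in a star rather than a chain.
# Hose bones 02 and 03 have locked swing and twist, while the claw and hose 01 are not locked.
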PhysicsAsset /BRCosmetics/Weapons/FORT_Melee/Pickaxe_Assemble_P_Male/Meshes/Assemble_P_Male_Axe_Front_End_Physics
bodies {
 Melee_TwoHanded_Handle [{shapeType=Capsule, length=110, radius=3, center=<10 0 0>, rotation=<rot 80 0 -5>}]
 claw_hoose_01_mstr_dyn [{shapeType=Capsule, length=12, radius=2, shapeName=claw_hoose_01_mstr_dyn_capsule, center=<16 0 -2>}]
 handle_hoose_01_dyn [{shapeType=Capsule, length=10, radius=2.5, shapeName=handle_hoose_01_dyn_capsule, center=<6 0 0>, rotation=<rot 90 0 0>}]
 handle_hoose_02_dyn [{shapeType=Capsule, length=10, radius=2, shapeName=handle_hoose_02_dyn_capsule, center=<2 0 0>}]
 handle_hoose_03_dyn [{shapeType=Capsule, length=10, radius=2, shapeName=handle_hoose_03_dyn_capsule, center=<2 0 0>}]
}
constraints [
 {bone1Name=claw_hoose_01_mstr_dyn, bone2Name=Melee_TwoHanded_Handle, swing1LimitDegrees=5, swing2LimitDegrees=5, twistLimitDegrees=1}
 {
  bone1Name=handle_hoose_02_dyn
  bone2Name=Melee_TwoHanded_Handle
  swing1LimitDegrees=30
  swing1Motion=Locked
  swing2LimitDegrees=30
  swing2Motion=Locked
  twistLimitDegrees=1
  twistMotion=Locked
 }
 {
  bone1Name=handle_hoose_03_dyn
  bone2Name=Melee_TwoHanded_Handle
  swing1LimitDegrees=30
  swing1Motion=Locked
  swing2LimitDegrees=30
  swing2Motion=Locked
  twistLimitDegrees=1
  twistMotion=Locked
 }
 {bone1Name=handle_hoose_01_dyn, bone2Name=Melee_TwoHanded_Handle, swing1LimitDegrees=65, swing2LimitDegrees=65, twistLimitDegrees=1}
]
