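# Animation Blueprint for the Assemble_P_Male pickaxe that adds secondary motion and a spinning cog to the weapon.
# Two RigidBody simulations, using front-end and regular physics assets, are blended by bIsInFrontEnd.
# The result feeds spring-driven AnimDynamics on arrow_dyn, then an additive rotation of cog_dyn_01.
# Each update, SpinFunction Idle advances a roll angle by fixed step plus velocity term, wrapping at 720, to drive that cog.
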AnimBlueprint /BRCosmetics/Weapons/FORT_Melee/Pickaxe_Assemble_P_Male/Meshes/Assemble_P_Male_Axe_AnimBP
target_skeleton=&/Assemble_P_Male_Axe_Skeleton
graphs [
 AnimGraph {nodes [
   Copy Pose From Mesh() {bUseAttachedParent=true}
   n1 = RigidBody(ComponentPose=Local To Component(LocalPose=Input Pose())) {
    bClampLinearTranslationLimitToRefPose=true
    bOverrideWorldGravity=true
    overridePhysicsAsset=&/Assemble_P_Male_Axe_Front_End_Physics
    simulationSpace=BaseBoneSpace
    alphaBoolBlend {blendOption=Linear}
    baseBoneRef {boneName=b_AR_Root}
    componentAppliedLinearAccClamp=<3000 3000 3500>
    componentLinearAccScale=<0.3 0.3 0.5>
    componentLinearVelScale=<0.15 0.15 0.25>
    overrideWorldGravity {z=-980}
   }
   n2 = Component To Local(ComponentPose=@n1)
   n3 = RigidBody(ComponentPose=Local To Component(LocalPose=Use cached pose 'copypose_cache'())) {
    bClampLinearTranslationLimitToRefPose=true
    bOverrideWorldGravity=true
    overridePhysicsAsset=&/Assemble_P_Male_Axe_Physics
    simulationSpace=BaseBoneSpace
    alphaBoolBlend {blendOption=Linear}
    baseBoneRef {boneName=b_AR_Root}
    componentAppliedLinearAccClamp=<3000 3000 3500>
    componentLinearAccScale=<0.3 0.3 0.5>
    componentLinearVelScale=<0.15 0.15 0.25>
    overrideWorldGravity {z=-980}
   }
   n4 = Component To Local(ComponentPose=@n3)
   n5 = Get bIsInFrontEnd()
   n6 = Anim Dynamics - Bone: arrow_dyn(
    ComponentPose=Local To Component(LocalPose=Blend Poses by bool(BlendPose_0=@n2, BlendPose_1=@n4, bActiveValue=@n5) {blendType=Linear})
    Alpha=1
    ExternalForce="0,0,0"
   ) {
    angularDampingOverride=0.9999
    angularSpringConstant=75
    bAngularSpring=true
    bGravityOverrideInSimSpace=true
    bLinearSpring=true
    bOverrideAngularDamping=true
    bOverrideLinearDamping=true
    gravityScale=0.3
    linearDampingOverride=0.9999
    alphaBoolBlend {blendOption=Linear}
    boundBone {boneName=arrow_dyn}
    componentLinearAccScale=<0.35 0.35 0.35>
    componentLinearVelScale=<0.35 0.35 0.35>
    physicsBodyDefinitions [{
      sphereCollisionRadius=0
      boundBone {boneName=arrow_dyn}
      boxExtents=<5 5 5>
      constraintSetup {angularLimitsMax=<2 50 0>, angularLimitsMin=<0 -30 0>, angularTarget=<0 0 0>}
      localJointOffset=<0 0 -1>
     }]
   }
   n7 = "Transform (Modify) Bone - Bone: cog_dyn_01"!(ComponentPose=@n6, bAlphaBoolEnabled:bool=True, Translation="0,0,0", Rotation=Get CrankRotator() -> (CrankRotator:Rotator), Scale="1,1,1") -> (Pose) {
    alphaInputType=Bool
    rotationMode=BMM_Additive
    rotationSpace=BCS_BoneSpace
    alphaBoolBlend {blendOption=Linear}
    boneToModify {boneName=cog_dyn_01}
   }
   Output Pose(Result=Component To Local(ComponentPose=@n7)) {layerGroup=None}
   Save cached pose 'copypose_cache'(Pose=Copy Pose From Mesh() {bUseAttachedParent=true}) {cacheName=copypose_cache}
  ]}
 EventGraph {nodes [
   Event BlueprintUpdateAnimation()
   > n1 = Sequence!(execute) -> (then_0, then_1, then_2:Exec)
   > n2 = SpinFunction Idle!(execute, self, Spin:double=Get crankSpin() -> (crankSpin:double), Speed:double=1.25) -> (then, Output:double)
   > n3 = Set crankSpin(crankSpin:double=@n2) -> (Output_Get:double)
   > Set CrankRotator(CrankRotator:Rotator=MakeRotator(Roll=@n3)) -> (Output_Get:Rotator)
   n4 = Branch(execute=@n1.then_1, Condition=Get bIsAccelerating2D())
   > Set Is_Accelerating_EG(Is_Accelerating_EG:bool=true) -> (Output_Get:bool)
   Set Is_Accelerating_EG(execute=@n4.else, Is_Accelerating_EG=false)
   TryGetPawnOwner()
  ]}
 SpinFunction Idle {nodes [
   n1 = SpinFunction Idle!() -> (then, Spin:double=~, Speed:double=~) {is=title K2Node_FunctionEntry}
   > n2 = Set spinVar(spinVar:double=@n1) -> (Output_Get:double)
   > n3 = Branch(Condition="float >= float"(A=@n2, B=720) {is=title K2Node_CallFunction})
   > n4 = Set spinVar(spinVar=float - float(A=Get spinVar() -> (spinVar:double), B=720) {is=title K2Node_CallFunction})
   n5 = Get deltaTimeRemap() -> (deltaTimeRemap:double)
   n6 = Get velocity() -> (velocity:double)
   n7 = Get velocityMult() -> (velocityMult:double)
   n8 = Set spinVar(
    execute=@n3.else
    spinVar=float + float!(A=Get spinVar(), B=float * float(A=@n1.Speed, B=@n5), C:double=float * float(A=@n6, B=@n7)) -> (ReturnValue) {is=title K2Node_CommutativeAssociativeBinaryOperator}
   )
   Return Node!(execute=[@n4.then, @n8.then], Output:double=@n8)
  ]}
]
variables [
 CrankRotator {type=Rotator}
 crankSpin {type=number}
 deltaTimeRemap {type=number, default=0.15}
 Is_Accelerating_EG {type=boolean}
 spinVar {type=number}
 velocity {type=number}
 velocityMult {type=number}
]
